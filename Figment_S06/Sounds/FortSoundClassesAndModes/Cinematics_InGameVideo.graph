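# Sound class for in-game cinematic video audio in Figment Season 6, parented to the RenderedCinematics class.
# It lowers volume to under a third, disables reverb and sets sounds to always play so they avoid being culled by concurrency or priority.
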SoundClass /Figment_S06/Sounds/FortSoundClassesAndModes/Cinematics_InGameVideo
parent=/Game/Sounds/FortSoundClassesAndModes/RenderedCinematics
properties {properties {bAlwaysPlay=true, bReverb=false, volume=0.3}}
