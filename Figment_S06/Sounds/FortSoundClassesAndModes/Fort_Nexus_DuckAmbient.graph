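# Fort_Nexus_DuckAmbient is a sound class under Fort_SFX_Nexus for Figment S06 sounds that, going by its name, duck ambient audio.
# While sounds in this class play, passive sound mix modifiers apply two mixes: its own DuckAmbient mix and the mapwide events mix.
# Both modifiers have a zero minimum volume threshold.
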SoundClass /Figment_S06/Sounds/FortSoundClassesAndModes/Fort_Nexus_DuckAmbient
parent=&/Fort_SFX_Nexus
properties {passiveSoundMixModifiers [
  {minVolumeThreshold=0, soundMix=&/Fort_Nexus_DuckAmbient_MixMod}
  {minVolumeThreshold=0, soundMix=/Figment_S06/Athena/Sounds/Events/SoundClasses/Fort_Events_Mapwide_PSM}
 ]}
