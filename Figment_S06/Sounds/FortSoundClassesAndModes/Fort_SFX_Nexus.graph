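# Fort_SFX_Nexus is a SoundClass that groups the Nexus sound effects in Figment_S06's Fort sound hierarchy.
# It sits under the SFX class and parents three child classes: Ambient, Transition and DuckAmbient.
# It overrides no properties of its own, serving as a grouping node in the class tree.
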SoundClass /Figment_S06/Sounds/FortSoundClassesAndModes/Fort_SFX_Nexus
parent=&SFX
children [
 &/Fort_Nexus_Ambient
 &/Fort_Nexus_Transition
 &/Fort_Nexus_DuckAmbient
]
