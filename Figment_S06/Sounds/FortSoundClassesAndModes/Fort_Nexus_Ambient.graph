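# Fort_Nexus_Ambient is a SoundClass in the Figment_S06 sound classes and modes folder, named for Nexus ambient audio.
# It is a child of the Fort_SFX_Nexus sound class and shows no property overrides of its own.
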SoundClass /Figment_S06/Sounds/FortSoundClassesAndModes/Fort_Nexus_Ambient
parent=&/Fort_SFX_Nexus
